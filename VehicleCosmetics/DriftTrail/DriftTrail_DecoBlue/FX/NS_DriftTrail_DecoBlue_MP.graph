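# Niagara system for the DecoBlue vehicle drift trail cosmetic (MP variant), emitting effects behind the wheels.
# A cone emitter bursts mesh and sprite particles with game-mode-dependent sine/cosine material pulses, while two sprite emitters spawn drifting particles and a trail emitter spawns ribbons.
# Cone and sprite particles use partition indices to pick skeletal mesh sockets and mask per-tire FX.
# Painted-variant overrides set each color, often with alpha fade curves.
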NiagaraSystem /VehicleCosmetics/DriftTrail/DriftTrail_DecoBlue/FX/NS_DriftTrail_DecoBlue_MP
dependencies {
 usedDataInterfaces [
  &NiagaraDataInterfaceColorCurve
  &NiagaraDataInterfaceCurve
  &NiagaraDataInterfaceSpriteRendererInfo
  &NiagaraDataInterfaceSkeletalMesh
 ]
 usedDynamicInputs []
 usedModules [
  &self:SystemSpawnScript.SystemScriptSource.SystemScriptGraph.NiagaraNodeAssignment_0.SetVariables_#1
  &FN_UnpackVehicleFlags
  /VehicleCosmetics/FX/Niagara/Modules/VC_SetVehicleSpeedAttribs
  &SystemState
  /Game/Effects/Niagara/Modules/Vehicle/FN_VehicleFlags_Cheatcodes
 ]
}
emitters {
 NE_VC_Cone {
  emitterUpdateScript [
   &EmitterState(
    Inactive Response=::NewEnumerator1
    Life Cycle Mode=::NewEnumerator1
    Loop Behavior=::NewEnumerator1
    Loop Duration Mode=::NewEnumerator1
    UseLoopDelay=false
   )
   &SpawnRate(
    SpawnRate=/VehicleCosmetics/FX/Niagara/Modules/VC_Mask_Floats(
     FrontEnd Value:NiagaraFloat=200
     blsFrontEndPreview:NiagaraBool=false
     SpeedMin Value:NiagaraFloat=120
     SpeedMax Value:NiagaraFloat=180
     Speed Alpha:NiagaraFloat=$System.Speed_Alpha
    )
    Use Spawn Probability=false
   ) {enabled=false}
   &SpawnBurst_Instantaneous(Spawn Count=4, Use Spawn Probability=false, UseLoopCountLimit=false)
   /SetVariables(Emitter.dragon:NiagaraBool=&Boolean_NotOperation(Bool=$User.bIsPainted)) {isSetParameters=true}
  ]
  particleSpawnScript [
   &VC_PaintedVariantsOverride(
    Black=<rgba 0.04 0.04 0.04 1>
    BurntSienna=<rgba 0.16 0.0412 0.0016 1>
    Cobalt=<rgba 0.095 0.1662 0.95 1>
    Crimson=<rgba 0.5 0.025 0.0329 1>
    ForestGreen=<rgba 0.008 0.4 0.0341 1>
    Grey=<rgba 0.23 0.23 0.23 1>
    Lime=<rgba 0.7793 1.4 0.07 1>
    Orange=<rgba 0.93 0.2883 0.0744 1>
    Pink=<rgba 0.7813 0.24 0.8 1>
    Purple=<rgba 0.1566 0.0076 0.38 1>
    R1 Black:LinearColor=<rgba 0.03 0.03 0.03 1>
    R1 BurntSienna:LinearColor=<rgba 0.12 0.021 0.0012 1>
    R1 Cobalt:LinearColor=<rgba 0.035 0.0904 0.7 1>
    R1 Crimson:LinearColor=<rgba 0.2 0.004 0.0073 1>
    R1 DefaultColor=<rgba 0.0355 0.8459 0.6583 1>
    R1 ForestGreen:LinearColor=<rgba 0.0025 0.25 0.019 1>
    R1 Grey:LinearColor=<rgba 0.1 0.1 0.1 1>
    R1 Lime:LinearColor=<rgba 0.303 0.6 0.006 1>
    R1 Orange:LinearColor=<rgba 0.6 0.159 0.012 1>
    R1 Pink:LinearColor=<rgba 0.4375 0.125 0.5 1>
    R1 Purple:LinearColor=<rgba 0.1157 0.0054 0.27 1>
    R1 Saffron:LinearColor=<rgba 0.5505 0.6 0.006 1>
    R1 SkyBlue:LinearColor=<rgba 0.005 0.2525 0.5 1>
    R1 TitaniumWhite:LinearColor=<rgba 0.5 0.5 0.5 1>
    R2 Black:LinearColor=<rgba 0.05 0.05 0.05 1>
    R2 BurntSienna:LinearColor=<rgba 0.4 0.115 0.02 1>
    R2 Cobalt:LinearColor=<rgba 0.04 0.1033 0.8 1>
    R2 Crimson:LinearColor=<rgba 0.3 0.006 0.0109 1>
    R2 DefaultColor:LinearColor=<rgba 0.0355 0.8459 0.6583 0.3>
    R2 ForestGreen:LinearColor=<rgba 0.008 0.4 0.0341 1>
    R2 Grey:LinearColor=<rgba 0.25 0.25 0.25 1>
    R2 Lime:LinearColor=<rgba 0.25 0.5 0 1>
    R2 Orange:LinearColor=<rgba 0.3 0.104 0.006 1>
    R2 Pink:LinearColor=<rgba 0.3767 0.12 0.4 1>
    R2 Purple:LinearColor=<rgba 0.2142 0.01 0.5 1>
    R2 Saffron:LinearColor=<rgba 0.4592 0.5 0.01 1>
    R2 SkyBlue:LinearColor=<rgba 0.005 0.2525 0.5 1>
    R2 TitaniumWhite:LinearColor=<rgba 0.4 0.4 0.4 1>
    R3 Black:LinearColor=<rgba 0.17 0.17 0.17 0.6>
    R3 BurntSienna:LinearColor=<rgba 0.8 0.1728 0.016 0.6>
    R3 Cobalt:LinearColor=<rgba 0.1782 0.3811 0.99 0.6>
    R3 Crimson:LinearColor=<rgba 0.86 0.0086 0.0228 0.6>
    R3 DefaultColor:LinearColor=<rgba 0.85 0.6 0.002 0.6>
    R3 ForestGreen:LinearColor=<rgba 0.0192 0.96 0.0192 0.6>
    R3 Grey:LinearColor=<rgba 0.65 0.65 0.65 0.6>
    R3 Lime:LinearColor=<rgba 0.657 0.9 0.09 0.6>
    R3 Orange:LinearColor=<rgba 0.95 0.4687 0.0475 0.6>
    R3 Pink:LinearColor=<rgba 0.8392 0.285 0.95 0.6>
    R3 Purple:LinearColor=<rgba 0.495 0.09 0.9 0.6>
    R3 Saffron:LinearColor=<rgba 0.8724 0.95 0.019 0.6>
    R3 SkyBlue:LinearColor=<rgba 0.1176 0.5488 0.98 0.6>
    R3 TitaniumWhite:LinearColor=<rgba 1 1 1 0.6>
    R4 Black:LinearColor=<rgba 0.21 0.21 0.21 1>
    R4 BurntSienna:LinearColor=<rgba 0.8 0.1728 0.016 1>
    R4 Cobalt:LinearColor=<rgba 0.3 0.375 0.6 1>
    R4 Crimson:LinearColor=<rgba 0.6 0.12 0.128 1>
    R4 DefaultColor:LinearColor=<rgba 0 0 0 1>
    R4 ForestGreen:LinearColor=<rgba 0.1767 0.8 0.12 1>
    R4 Gold:LinearColor=<rgba 0 0 0 1>
    R4 Grey:LinearColor=<rgba 0.7 0.7 0.7 1>
    R4 Lime:LinearColor=<rgba 0.584 0.8 0.08 1>
    R4 Orange:LinearColor=<rgba 0.9 0.555 0.072 1>
    R4 Pink:LinearColor=<rgba 0.855 0.38 0.95 1>
    R4 Purple:LinearColor=<rgba 0.413 0.126 0.7 1>
    R4 Saffron:LinearColor=<rgba 0.9 0.9 0.18 1>
    R4 SkyBlue:LinearColor=<rgba 0.1782 0.6517 0.99 1>
    R4 TitaniumWhite:LinearColor=<rgba 0.8 0.8 0.8 1>
    Saffron=<rgba 1 1 0.02 1>
    SkyBlue=<rgba 0.02 0.5263 1 1>
    TitaniumWhite=<rgba 1 1 1 1>
    Use R1 Painted Overrides=true
    Use R2 Painted Overrides=true
    Use R3 Painted Overrides=true
    Use R4 Painted Overrides=true
   )
   &InitializeParticle(
    Color=$Particles.Color
    Lifetime=2
    Mesh Scale Mode=::NewEnumerator3
    Mesh Uniform Scale=2.6
    Sprite Size=<80 80>
    Sprite Size Mode=::NewEnumerator4
    UsePositionOffset=false
    Write Material Random=true
   )
   &PartitionParticles()
   &SkeletalMeshLocation(
    Bone / Socket Index=&Add_Integer(A=$Particles.PartitionParticles.Partition, B=4)
    Bone / Socket Sampling Mode=::NewEnumerator5
    Interpolate Mesh Sampling=false
    MaskSpawnGroup=false
    Mesh Sampling Type=::NewEnumerator6
    Sampled Position Offset=&AddVectorToPosition(Position=<-95 0 -20>, Vector=$User.offset)
    Skeletal Mesh=$System.SkeletalMesh
   )
   &AddVelocity(Velocity=&VC_Mask_Vectors(
     FrontEnd Value:Vector3f=<-1500 0 0>
     blsFrontEndPreview:NiagaraBool=false
     SpeedMin Value:Vector3f=<-800 0 0>
     SpeedMax Value:Vector3f=<-1 0 0>
     Speed Alpha:NiagaraFloat=$System.Speed_Alpha
    )) {enabled=false}
   /SetVariables(
    Particles.dotposition:NiagaraPosition=&AddVectorToPosition(Position=$Particles.Position, Vector=<10 0 0>)
    Particles.dotsoffset=-5
   ) {isSetParameters=true}
  ]
  particleUpdateScript [
   &FN_DisableTireFX(
    Mask Index Attribute:NiagaraInt32=$Particles.PartitionParticles.Partition
    Mask Index FL=2
    Mask Index FR=3
    Mask Index RL=0
    Mask Index RR=1
   )
   &ParticleState(Kill Particles When Lifetime Has Elapsed=false)
   &SolveForcesAndVelocity(Clamp Velocity=false, Limit Acceleration=false) {enabled=false}
   &DynamicMaterialParameters(
    Index 1 Param 1=0.5
    Index 3 Param 1=&Cosine(
     Normalized Angle=&Add_Float(
      A=$Particles.NormalizedAge
      B=&SelectFloatFromArray(Direct Array Index=$Particles.PartitionParticles.Partition, Float Selection Array=DataInterface() {propertyValues {FloatData [0,0.8]}})
     )
     Period=&MergeHelper_FloatSwitch(0=1, 1=0.6, 2=0.5, 3=?, Index=$User.GameModeID, Num=::NewEnumerator1)
     Scale=&MergeHelper_FloatSwitch(0=3, 1=10, 2=10, 3=?, Index=$User.GameModeID, Num=::NewEnumerator1)
     Bias=0
    )
    Index 3 Param 2=&Cosine(
     Normalized Angle=&Add_Float(
      A=$Particles.NormalizedAge
      B=&SelectFloatFromArray(Direct Array Index=$Particles.PartitionParticles.Partition, Float Selection Array=DataInterface() {propertyValues {FloatData [0,0.2]}})
     )
     Period=&MergeHelper_FloatSwitch(0=1, 1=1, 2=0.5, 3=?, Index=$User.GameModeID, Num=::NewEnumerator1)
     Scale=&MergeHelper_FloatSwitch(0=2, 1=1, 2=3, 3=?, Index=$User.GameModeID, Num=::NewEnumerator1)
     Bias=0
    )
    Index 3 Param 3=&Sine(
     Normalized Angle=&Add_Float(
      A=$Particles.NormalizedAge
      B=&SelectFloatFromArray(Direct Array Index=$Particles.PartitionParticles.Partition, Float Selection Array=DataInterface() {propertyValues {FloatData [0,0.5]}})
     )
     Period=&MergeHelper_FloatSwitch(0=1, 1=0.6, 2=0.5, 3=?, Index=$User.GameModeID, Num=::NewEnumerator1)
     Scale=&MergeHelper_FloatSwitch(0=2, 1=4, 2=5, 3=?, Index=$User.GameModeID, Num=::NewEnumerator1)
     Bias=0
    )
    Module.Index1Float0Write=true
    Module.Index1Float1Write=true
    Module.Index1Float2Write=false
    Module.Index1Float3Write=false
    Module.Index3Float0Write=true
    Module.Index3Float1Write=true
    Module.Index3Float2Write=true
    Module.Index3Float3Write=false
    Write Parameter Index 0=::NewEnumerator0
    Write Parameter Index 1=::NewEnumerator1
    Write Parameter Index 3=::NewEnumerator1
   )
  ]
  rendererClasses [&NiagaraMeshRendererProperties, &NiagaraSpriteRendererProperties]
  renderers [
   {rendererClass=&NiagaraMeshRendererProperties, rendererIndex=0}
   {rendererClass=&NiagaraMeshRendererProperties, rendererIndex=1}
   {rendererClass=&NiagaraMeshRendererProperties, rendererIndex=2}
   {rendererClass=&NiagaraMeshRendererProperties, rendererIndex=3}
   {rendererClass=&NiagaraMeshRendererProperties, rendererIndex=4}
   {rendererClass=&NiagaraMeshRendererProperties, rendererIndex=5}
   {rendererClass=&NiagaraMeshRendererProperties, rendererIndex=6}
   {rendererClass=&NiagaraSpriteRendererProperties, rendererIndex=7}
  ]
 }
 NE_VC_Sprite {
  emitterUpdateScript [
   &EmitterState(Inactive Response=::NewEnumerator1, Life Cycle Mode=::NewEnumerator1, UseLoopDelay=false)
   &SpawnRate(Spawn Probability=0.6, SpawnRate=9, Use Spawn Probability=true)
  ]
  particleSpawnScript [
   &VC_PaintedVariantsOverride(Use R1 Painted Overrides=true) {enabled=false}
   &InitializeParticle(
    Lifetime Max=1.2
    Lifetime Mode=::NewEnumerator1
    Sprite Rotation Mode=::NewEnumerator2
    Sprite Size Mode=::NewEnumerator3
    Uniform Sprite Size=22
    UsePositionOffset=false
    Write Material Random=true
   )
   &PartitionParticles()
   &SkeletalMeshLocation(
    Bone / Socket Index=&Add_Integer(A=$Particles.PartitionParticles.Partition, B=4)
    Bone / Socket Sampling Mode=::NewEnumerator5
    Interpolate Mesh Sampling=false
    MaskSpawnGroup=false
    Mesh Sampling Type=::NewEnumerator6
    Skeletal Mesh=$System.SkeletalMesh
   )
   &ShapeLocation(
    MaskSpawnGroup=false
    Offset=&AddVectorToPosition(Position=<-170 0 5>, Vector=$User.offset)
    Offset Coordinate Space=::Local
    Offset Mode=::NewEnumerator0
    Override Seed=false
    Sphere Radius=20
    Sphere Surface Distribution=0.6
    UseOwnerScale=true
   )
   &AddVelocity(Velocity=<-300 0 0>)
  ]
  particleUpdateScript [
   &VC_PaintedVariantsOverride(
    Black=&ColorFromCurve(ColorCurve=DataInterface() {propertyValues {
       AlphaCurve {keys [<at 0.038 0>, <at 0.4646 1>, {arriveTangent=-1.9355, interpMode=RCIM_Cubic, leaveTangent=-1.9355, time=0.8154, value=1}, <at 0.9813 0>]}
       BlueCurve {keys [<at 0 0.015 RCIM_Cubic>]}
       GreenCurve {keys [<at 0 0.015 RCIM_Cubic>]}
       RedCurve {keys [<at 0 0.015 RCIM_Cubic>]}
      }})
    BurntSienna=&ColorFromCurve(ColorCurve=DataInterface() {propertyValues {
       AlphaCurve {keys [<at 0.038 0>, <at 0.4646 1>, {arriveTangent=-1.9355, interpMode=RCIM_Cubic, leaveTangent=-1.9355, time=0.8154, value=1}, <at 0.9813 0>]}
       BlueCurve {keys [<at 0 0.0025 RCIM_Cubic>]}
       GreenCurve {keys [<at 0 0.0136 RCIM_Cubic>]}
       RedCurve {keys [<at 0 0.05 RCIM_Cubic>]}
      }})
    Cobalt=&ColorFromCurve(ColorCurve=DataInterface() {propertyValues {
       AlphaCurve {keys [<at 0.038 0>, <at 0.4646 1>, {arriveTangent=-1.9355, interpMode=RCIM_Cubic, leaveTangent=-1.9355, time=0.8154, value=1}, <at 0.9813 0>]}
       BlueCurve {keys [<at 0 0.15 RCIM_Cubic>]}
       GreenCurve {keys [<at 0 0.0262 RCIM_Cubic>]}
       RedCurve {keys [<at 0 0.015 RCIM_Cubic>]}
      }})
    Crimson=&ColorFromCurve(ColorCurve=DataInterface() {propertyValues {
       AlphaCurve {keys [<at 0.038 0>, <at 0.4646 1>, {arriveTangent=-1.9355, interpMode=RCIM_Cubic, leaveTangent=-1.9355, time=0.8154, value=1}, <at 0.9813 0>]}
       BlueCurve {keys [<at 0 0.0026 RCIM_Cubic>]}
       GreenCurve {keys [<at 0 0.001 RCIM_Cubic>]}
       RedCurve {keys [<at 0 0.1 RCIM_Cubic>]}
      }})
    DefaultColor=&ColorFromCurve(ColorCurve=DataInterface() {propertyValues {
       AlphaCurve {keys [<at 0.038 0>, <at 0.4646 1>, {arriveTangent=-1.9355, interpMode=RCIM_Cubic, leaveTangent=-1.9355, time=0.8154, value=1}, <at 0.9813 0>]}
       BlueCurve {keys [<at 0 0.2812 RCIM_Cubic>]}
       GreenCurve {keys [<at 0 0.0659 RCIM_Cubic>]}
       RedCurve {keys [<at 0 0.0659 RCIM_Cubic>]}
      }})
    ForestGreen=&ColorFromCurve(ColorCurve=DataInterface() {propertyValues {
       AlphaCurve {keys [<at 0.038 0>, <at 0.4646 1>, {arriveTangent=-1.9355, interpMode=RCIM_Cubic, leaveTangent=-1.9355, time=0.8154, value=1}, <at 0.9813 0>]}
       BlueCurve {keys [<at 0 0.0041 RCIM_Cubic>]}
       GreenCurve {keys [<at 0 0.06 RCIM_Cubic>]}
       RedCurve {keys [<at 0 0.0012 RCIM_Cubic>]}
      }})
    Grey=&ColorFromCurve(ColorCurve=DataInterface() {propertyValues {
       AlphaCurve {keys [<at 0.038 0>, <at 0.4646 1>, {arriveTangent=-1.9355, interpMode=RCIM_Cubic, leaveTangent=-1.9355, time=0.8154, value=1}, <at 0.9813 0>]}
       BlueCurve {keys [<at 0 0.04 RCIM_Cubic>]}
       GreenCurve {keys [<at 0 0.04 RCIM_Cubic>]}
       RedCurve {keys [<at 0 0.04 RCIM_Cubic>]}
      }})
    Lime=&ColorFromCurve(ColorCurve=DataInterface() {propertyValues {
       AlphaCurve {keys [<at 0.038 0>, <at 0.4646 1>, {arriveTangent=-1.9355, interpMode=RCIM_Cubic, leaveTangent=-1.9355, time=0.8154, value=1}, <at 0.9813 0>]}
       BlueCurve {keys [<at 0 0.0022 RCIM_Cubic>]}
       GreenCurve {keys [<at 0 0.11 RCIM_Cubic>]}
       RedCurve {keys [<at 0 0.0651 RCIM_Cubic>]}
      }})
    Orange=&ColorFromCurve(ColorCurve=DataInterface() {propertyValues {
       AlphaCurve {keys [<at 0.038 0>, <at 0.4646 1>, {arriveTangent=-1.9355, interpMode=RCIM_Cubic, leaveTangent=-1.9355, time=0.8154, value=1}, <at 0.9813 0>]}
       BlueCurve {keys [<at 0 0.0055 RCIM_Cubic>]}
       GreenCurve {keys [<at 0 0.0334 RCIM_Cubic>]}
       RedCurve {keys [<at 0 0.11 RCIM_Cubic>]}
      }})
    Pink=&ColorFromCurve(ColorCurve=DataInterface() {propertyValues {
       AlphaCurve {keys [<at 0.038 0>, <at 0.4646 1>, {arriveTangent=-1.9355, interpMode=RCIM_Cubic, leaveTangent=-1.9355, time=0.8154, value=1}, <at 0.9813 0>]}
       BlueCurve {keys [<at 0 0.1 RCIM_Cubic>]}
       GreenCurve {keys [<at 0 0.035 RCIM_Cubic>]}
       RedCurve {keys [<at 0 0.0946 RCIM_Cubic>]}
      }})
    Purple=&ColorFromCurve(ColorCurve=DataInterface() {propertyValues {
       AlphaCurve {keys [<at 0.038 0>, <at 0.4646 1>, {arriveTangent=-1.9355, interpMode=RCIM_Cubic, leaveTangent=-1.9355, time=0.8154, value=1}, <at 0.9813 0>]}
       BlueCurve {keys [<at 0 0.06 RCIM_Cubic>]}
       GreenCurve {keys [<at 0 0.0012 RCIM_Cubic>]}
       RedCurve {keys [<at 0 0.0257 RCIM_Cubic>]}
      }})
    R1 Black:LinearColor=&ColorFromCurve(ColorCurve=DataInterface() {propertyValues {
       AlphaCurve {keys [<at 0.0271 0>,<at 0.3529 1>,<at 0.6981 1 RCIM_Cubic>,<at 0.9415 0 RCIM_Cubic>]}
       BlueCurve {keys [<at 0 0.03 RCIM_Cubic>]}
       GreenCurve {keys [<at 0 0.03 RCIM_Cubic>]}
       RedCurve {keys [<at 0 0.03 RCIM_Cubic>]}
      }})
    R1 BurntSienna:LinearColor=&ColorFromCurve(ColorCurve=DataInterface() {propertyValues {
       AlphaCurve {keys [<at 0.0271 0>,<at 0.3529 1>,<at 0.6981 1 RCIM_Cubic>,<at 0.9415 0 RCIM_Cubic>]}
       BlueCurve {keys [<at 0 0.006 RCIM_Cubic>]}
       GreenCurve {keys [<at 0 0.0345 RCIM_Cubic>]}
       RedCurve {keys [<at 0 0.12 RCIM_Cubic>]}
      }})
    R1 Cobalt:LinearColor=&ColorFromCurve(ColorCurve=DataInterface() {propertyValues {
       AlphaCurve {keys [<at 0.0271 0>,<at 0.3529 1>,<at 0.6981 1 RCIM_Cubic>,<at 0.9415 0 RCIM_Cubic>]}
       BlueCurve {keys [<at 0 0.8 RCIM_Cubic>]}
       GreenCurve {keys [<at 0 0.164 RCIM_Cubic>]}
       RedCurve {keys [<at 0 0.08 RCIM_Cubic>]}
      }})
    R1 Crimson:LinearColor=&ColorFromCurve(ColorCurve=DataInterface() {propertyValues {
       AlphaCurve {keys [<at 0.0271 0>,<at 0.3529 1>,<at 0.6981 1 RCIM_Cubic>,<at 0.9415 0 RCIM_Cubic>]}
       BlueCurve {keys [<at 0 0.0109 RCIM_Cubic>]}
       GreenCurve {keys [<at 0 0.006 RCIM_Cubic>]}
       RedCurve {keys [<at 0 0.3 RCIM_Cubic>]}
      }})
    R1 DefaultColor=&ColorFromCurve(ColorCurve=DataInterface() {propertyValues {
       AlphaCurve {keys [<at 0.0271 0>,<at 0.3529 1>,<at 0.6981 1 RCIM_Cubic>,<at 0.9415 0 RCIM_Cubic>]}
       BlueCurve {keys [<at 0 0.8957 RCIM_Cubic>]}
       GreenCurve {keys [<at 0 0.513 RCIM_Cubic>]}
       RedCurve {keys [<at 0 0.0255 RCIM_Cubic>]}
      }})
    R1 ForestGreen:LinearColor=&ColorFromCurve(ColorCurve=DataInterface() {propertyValues {
       AlphaCurve {keys [<at 0.0271 0>,<at 0.3529 1>,<at 0.6981 1 RCIM_Cubic>,<at 0.9415 0 RCIM_Cubic>]}
       BlueCurve {keys [<at 0 0.006 RCIM_Cubic>]}
       GreenCurve {keys [<at 0 0.1 RCIM_Cubic>]}
       RedCurve {keys [<at 0 0.001 RCIM_Cubic>]}
      }})
    R1 Grey:LinearColor=&ColorFromCurve(ColorCurve=DataInterface() {propertyValues {
       AlphaCurve {keys [<at 0.0271 0>,<at 0.3529 1>,<at 0.6981 1 RCIM_Cubic>,<at 0.9415 0 RCIM_Cubic>]}
       BlueCurve {keys [<at 0 0.1 RCIM_Cubic>]}
       GreenCurve {keys [<at 0 0.1 RCIM_Cubic>]}
       RedCurve {keys [<at 0 0.1 RCIM_Cubic>]}
      }})
    R1 Lime:LinearColor=&ColorFromCurve(ColorCurve=DataInterface() {propertyValues {
       AlphaCurve {keys [<at 0.0271 0>,<at 0.3529 1>,<at 0.6981 1 RCIM_Cubic>,<at 0.9415 0 RCIM_Cubic>]}
       BlueCurve {keys [<at 0 0.005 RCIM_Cubic>]}
       GreenCurve {keys [<at 0 0.25 RCIM_Cubic>]}
       RedCurve {keys [<at 0 0.1357 RCIM_Cubic>]}
      }})
    R1 Orange:LinearColor=&ColorFromCurve(ColorCurve=DataInterface() {propertyValues {
       AlphaCurve {keys [<at 0.0271 0>,<at 0.3529 1>,<at 0.6981 1 RCIM_Cubic>,<at 0.9415 0 RCIM_Cubic>]}
       BlueCurve {keys [<at 0 0.005 RCIM_Cubic>]}
       GreenCurve {keys [<at 0 0.1071 RCIM_Cubic>]}
       RedCurve {keys [<at 0 0.25 RCIM_Cubic>]}
      }})
    R1 Pink:LinearColor=&ColorFromCurve(ColorCurve=DataInterface() {propertyValues {
       AlphaCurve {keys [<at 0.0271 0>,<at 0.3529 1>,<at 0.6981 1 RCIM_Cubic>,<at 0.9415 0 RCIM_Cubic>]}
       BlueCurve {keys [<at 0 0.25 RCIM_Cubic>]}
       GreenCurve {keys [<at 0 0.05 RCIM_Cubic>]}
       RedCurve {keys [<at 0 0.2167 RCIM_Cubic>]}
      }})
    R1 Purple:LinearColor=&ColorFromCurve(ColorCurve=DataInterface() {propertyValues {
       AlphaCurve {keys [<at 0.0271 0>,<at 0.3529 1>,<at 0.6981 1 RCIM_Cubic>,<at 0.9415 0 RCIM_Cubic>]}
       BlueCurve {keys [<at 0 0.3 RCIM_Cubic>]}
       GreenCurve {keys [<at 0 0.006 RCIM_Cubic>]}
       RedCurve {keys [<at 0 0.1285 RCIM_Cubic>]}
      }})
    R1 Saffron:LinearColor=&ColorFromCurve(ColorCurve=DataInterface() {propertyValues {
       AlphaCurve {keys [<at 0.0271 0>,<at 0.3529 1>,<at 0.6981 1 RCIM_Cubic>,<at 0.9415 0 RCIM_Cubic>]}
       BlueCurve {keys [<at 0 0.005 RCIM_Cubic>]}
       GreenCurve {keys [<at 0 0.25 RCIM_Cubic>]}
       RedCurve {keys [<at 0 0.25 RCIM_Cubic>]}
      }})
    R1 SkyBlue:LinearColor=&ColorFromCurve(ColorCurve=DataInterface() {propertyValues {
       AlphaCurve {keys [<at 0.0271 0>,<at 0.3529 1>,<at 0.6981 1 RCIM_Cubic>,<at 0.9415 0 RCIM_Cubic>]}
       BlueCurve {keys [<at 0 0.4 RCIM_Cubic>]}
       GreenCurve {keys [<at 0 0.169 RCIM_Cubic>]}
       RedCurve {keys [<at 0 0.004 RCIM_Cubic>]}
      }})
    R1 TitaniumWhite:LinearColor=&ColorFromCurve(ColorCurve=DataInterface() {propertyValues {
       AlphaCurve {keys [<at 0.0271 0>,<at 0.3529 1>,<at 0.6981 1 RCIM_Cubic>,<at 0.9415 0 RCIM_Cubic>]}
       BlueCurve {keys [<at 0 0.2 RCIM_Cubic>]}
       GreenCurve {keys [<at 0 0.2 RCIM_Cubic>]}
       RedCurve {keys [<at 0 0.2 RCIM_Cubic>]}
      }})
    Saffron=&ColorFromCurve(ColorCurve=DataInterface() {propertyValues {
       AlphaCurve {keys [<at 0.038 0>, <at 0.4646 1>, {arriveTangent=-1.9355, interpMode=RCIM_Cubic, leaveTangent=-1.9355, time=0.8154, value=1}, <at 0.9813 0>]}
       BlueCurve {keys [<at 0 0.02 RCIM_Cubic>]}
       GreenCurve {keys [<at 0 0.1 RCIM_Cubic>]}
       RedCurve {keys [<at 0 0.1 RCIM_Cubic>]}
      }})
    SkyBlue=&ColorFromCurve(ColorCurve=DataInterface() {propertyValues {
       AlphaCurve {keys [<at 0.038 0>, <at 0.4646 1>, {arriveTangent=-1.9355, interpMode=RCIM_Cubic, leaveTangent=-1.9355, time=0.8154, value=1}, <at 0.9813 0>]}
       BlueCurve {keys [<at 0 0.15 RCIM_Cubic>]}
       GreenCurve {keys [<at 0 0.0634 RCIM_Cubic>]}
       RedCurve {keys [<at 0 0.0015 RCIM_Cubic>]}
      }})
    TitaniumWhite=&ColorFromCurve(ColorCurve=DataInterface() {propertyValues {
       AlphaCurve {keys [<at 0.038 0>, <at 0.4646 1>, {arriveTangent=-1.9355, interpMode=RCIM_Cubic, leaveTangent=-1.9355, time=0.8154, value=1}, <at 0.9813 0>]}
       BlueCurve {keys [<at 0 0.1 RCIM_Cubic>]}
       GreenCurve {keys [<at 0 0.1 RCIM_Cubic>]}
       RedCurve {keys [<at 0 0.1 RCIM_Cubic>]}
      }})
    Use R1 Painted Overrides=true
   ) {module=VC_PaintedVariantsOverride001}
   &FN_DisableTireFX(
    Mask Index Attribute:NiagaraInt32=$Particles.PartitionParticles.Partition
    Mask Index FL=2
    Mask Index FR=3
    Mask Index RL=0
    Mask Index RR=1
   )
   &InheritSourceMovement(ApplyForce=false, ApplyPosition=true, ApplyPositionFalloff=false)
   &ParticleState()
   &ScaleSpriteSize(Uniform Curve Sprite Scale=DataInterface() {propertyValues {Curve {keys [{tangentMode=RCTM_User, time=0, value=0.3}, <at 0.2 1>, <at 0.7 1>]}}})
   &SolveForcesAndVelocity(Clamp Velocity=false, Limit Acceleration=false)
   &SubUVAnimation(
    "Sprite Renderer "=DataInterface() {propertyValues {SpriteRenderer=&self:NE_VC_Sprite_0.NiagaraSpriteRendererProperties_4}}
    SubUV Animation Mode:ENiagara_SubUVLookupModeV2=::NewEnumerator1
    SubUV Curve=DataInterface() {propertyValues {Curve {defaultValue=0, keys [<at 0 0.4>,<at 1 1>]}}}
    UseEndFrame=false
    UseStartFrame=false
   )
  ]
  rendererClasses [&NiagaraSpriteRendererProperties]
  renderers [
   {rendererClass=&NiagaraSpriteRendererProperties}
   {rendererClass=&NiagaraSpriteRendererProperties, rendererIndex=1}
  ]
 }
 NE_VC_Sprite001 {
  emitterUpdateScript [
   &EmitterState(Inactive Response=::NewEnumerator1, Life Cycle Mode=::NewEnumerator1, UseLoopDelay=false)
   &SpawnRate(Spawn Probability=0.4, SpawnRate=10, Use Spawn Probability=true)
  ]
  particleSpawnScript [
   &VC_PaintedVariantsOverride() {enabled=false}
   &InitializeParticle(
    Color Mode=::NewEnumerator0
    Lifetime Max=1.2
    Lifetime Mode=::NewEnumerator1
    Sprite Rotation Mode=::NewEnumerator2
    Sprite Size Mode=::NewEnumerator1
    Uniform Sprite Size Max=4.5
    Uniform Sprite Size Min=3.5
    UsePositionOffset=false
    Write Material Random=true
   )
   &PartitionParticles()
   &SkeletalMeshLocation(
    Bone / Socket Index=&Add_Integer(A=$Particles.PartitionParticles.Partition, B=4)
    Bone / Socket Sampling Mode=::NewEnumerator5
    Interpolate Mesh Sampling=false
    MaskSpawnGroup=false
    Mesh Sampling Type=::NewEnumerator6
    Skeletal Mesh=$System.SkeletalMesh
   )
   &ShapeLocation(
    MaskSpawnGroup=false
    Offset=&AddVectorToPosition(Position=<-160 0 -2>, Vector=$User.offset)
    Offset Coordinate Space=::Local
    Offset Mode=::NewEnumerator0
    Override Seed=false
    Sphere Radius=20
    Sphere Surface Distribution=0.5
    UseOwnerScale=true
   )
   &AddVelocity(Velocity=<-300 0 0>)
   /SetVariables(Particles.glowscale:Vector2f=&Multiply_Vector2D_ByFloat(
     Float=1.2
     Vector2D=&Vector2DFromFloat(Value=&RandomRangeFloat(Minimum=7, Maximum=9))
    )) {isSetParameters=true}
  ]
  particleUpdateScript [
   &VC_PaintedVariantsOverride(
    Black=&ColorFromCurve(ColorCurve=DataInterface() {propertyValues {
       AlphaCurve {keys [<at 0 0>, {arriveTangent=1.583, interpMode=RCIM_Cubic, leaveTangent=1.583, time=0.2533, value=1}, <at 0.6317 1>, <at 0.9658 0 RCIM_Cubic>]}
       BlueCurve {keys [<at 0 3 RCIM_Cubic>]}
       GreenCurve {keys [<at 0 3 RCIM_Cubic>]}
       RedCurve {keys [<at 0 3 RCIM_Cubic>]}
      }})
    BurntSienna=&ColorFromCurve(ColorCurve=DataInterface() {propertyValues {
       AlphaCurve {keys [<at 0 0>, {arriveTangent=1.583, interpMode=RCIM_Cubic, leaveTangent=1.583, time=0.2533, value=1}, <at 0.6317 1>, <at 0.9658 0 RCIM_Cubic>]}
       BlueCurve {keys [<at 0 0 RCIM_Cubic>]}
       GreenCurve {keys [<at 0 0.6 RCIM_Cubic>]}
       RedCurve {keys [<at 0 3 RCIM_Cubic>]}
      }})
    Cobalt=&ColorFromCurve(ColorCurve=DataInterface() {propertyValues {
       AlphaCurve {keys [<at 0 0>, {arriveTangent=1.583, interpMode=RCIM_Cubic, leaveTangent=1.583, time=0.2533, value=1}, <at 0.6317 1>, <at 0.9658 0 RCIM_Cubic>]}
       BlueCurve {keys [<at 0 3 RCIM_Cubic>]}
       GreenCurve {keys [<at 0 1.125 RCIM_Cubic>]}
       RedCurve {keys [<at 0 0.75 RCIM_Cubic>]}
      }})
    Crimson=&ColorFromCurve(ColorCurve=DataInterface() {propertyValues {
       AlphaCurve {keys [<at 0 0>, {arriveTangent=1.583, interpMode=RCIM_Cubic, leaveTangent=1.583, time=0.2533, value=1}, <at 0.6317 1>, <at 0.9658 0 RCIM_Cubic>]}
       BlueCurve {keys [<at 0 0.0727 RCIM_Cubic>]}
       GreenCurve {keys [<at 0 0.04 RCIM_Cubic>]}
       RedCurve {keys [<at 0 2 RCIM_Cubic>]}
      }})
    DefaultColor=&ColorFromCurve(ColorCurve=DataInterface() {propertyValues {
       AlphaCurve {keys [<at 0 0>, {arriveTangent=1.583, interpMode=RCIM_Cubic, leaveTangent=1.583, time=0.2533, value=1}, <at 0.6317 1>, <at 0.9658 0 RCIM_Cubic>]}
       BlueCurve {keys [<at 0 0.8 RCIM_Cubic>]}
       GreenCurve {keys [<at 0 4.4 RCIM_Cubic>]}
       RedCurve {keys [<at 0 8 RCIM_Cubic>]}
      }})
    ForestGreen=&ColorFromCurve(ColorCurve=DataInterface() {propertyValues {
       AlphaCurve {keys [<at 0 0>, {arriveTangent=1.583, interpMode=RCIM_Cubic, leaveTangent=1.583, time=0.2533, value=1}, <at 0.6317 1>, <at 0.9658 0 RCIM_Cubic>]}
       BlueCurve {keys [<at 0 0.312 RCIM_Cubic>]}
       GreenCurve {keys [<at 0 3 RCIM_Cubic>]}
       RedCurve {keys [<at 0 0.12 RCIM_Cubic>]}
      }})
    Grey=&ColorFromCurve(ColorCurve=DataInterface() {propertyValues {
       AlphaCurve {keys [<at 0 0>, {arriveTangent=1.583, interpMode=RCIM_Cubic, leaveTangent=1.583, time=0.2533, value=1}, <at 0.6317 1>, <at 0.9658 0 RCIM_Cubic>]}
       BlueCurve {keys [<at 0 4 RCIM_Cubic>]}
       GreenCurve {keys [<at 0 4 RCIM_Cubic>]}
       RedCurve {keys [<at 0 4 RCIM_Cubic>]}
      }})
    Lime=&ColorFromCurve(ColorCurve=DataInterface() {propertyValues {
       AlphaCurve {keys [<at 0 0>, {arriveTangent=1.583, interpMode=RCIM_Cubic, leaveTangent=1.583, time=0.2533, value=1}, <at 0.6317 1>, <at 0.9658 0 RCIM_Cubic>]}
       BlueCurve {keys [<at 0 0 RCIM_Cubic>]}
       GreenCurve {keys [<at 0 3 RCIM_Cubic>]}
       RedCurve {keys [<at 0 1.75 RCIM_Cubic>]}
      }})
    Orange=&ColorFromCurve(ColorCurve=DataInterface() {propertyValues {
       AlphaCurve {keys [<at 0 0>, {arriveTangent=1.583, interpMode=RCIM_Cubic, leaveTangent=1.583, time=0.2533, value=1}, <at 0.6317 1>, <at 0.9658 0 RCIM_Cubic>]}
       BlueCurve {keys [<at 0 0 RCIM_Cubic>]}
       GreenCurve {keys [<at 0 1.25 RCIM_Cubic>]}
       RedCurve {keys [<at 0 3 RCIM_Cubic>]}
      }})
    Pink=&ColorFromCurve(ColorCurve=DataInterface() {propertyValues {
       AlphaCurve {keys [<at 0 0>, {arriveTangent=1.583, interpMode=RCIM_Cubic, leaveTangent=1.583, time=0.2533, value=1}, <at 0.6317 1>, <at 0.9658 0 RCIM_Cubic>]}
       BlueCurve {keys [<at 0 3 RCIM_Cubic>]}
       GreenCurve {keys [<at 0 0.9 RCIM_Cubic>]}
       RedCurve {keys [<at 0 2.65 RCIM_Cubic>]}
      }})
    Purple=&ColorFromCurve(ColorCurve=DataInterface() {propertyValues {
       AlphaCurve {keys [<at 0 0>, {arriveTangent=1.583, interpMode=RCIM_Cubic, leaveTangent=1.583, time=0.2533, value=1}, <at 0.6317 1>, <at 0.9658 0 RCIM_Cubic>]}
       BlueCurve {keys [<at 0 2 RCIM_Cubic>]}
       GreenCurve {keys [<at 0 0.04 RCIM_Cubic>]}
       RedCurve {keys [<at 0 0.8567 RCIM_Cubic>]}
      }})
    R1 Black:LinearColor=&ColorFromCurve(ColorCurve=DataInterface() {propertyValues {
       AlphaCurve {keys [
         <at 0 0>
         {arriveTangent=0.3166, interpMode=RCIM_Cubic, leaveTangent=0.3166, time=0.2533, value=0.2}
         <at 0.6317 0.2>
         <at 0.9658 0 RCIM_Cubic>
        ]}
       BlueCurve {keys [<at 0 0.1927 RCIM_Cubic>]}
       GreenCurve {keys [<at 0 0.1927 RCIM_Cubic>]}
       RedCurve {keys [<at 0 0.1927 RCIM_Cubic>]}
      }})
    R1 BurntSienna:LinearColor=&ColorFromCurve(ColorCurve=DataInterface() {propertyValues {
       AlphaCurve {keys [
         <at 0 0>
         {arriveTangent=0.3166, interpMode=RCIM_Cubic, leaveTangent=0.3166, time=0.2533, value=0.2}
         <at 0.6317 0.2>
         <at 0.9658 0 RCIM_Cubic>
        ]}
       BlueCurve {keys [<at 0 0 RCIM_Cubic>]}
       GreenCurve {keys [<at 0 0.491 RCIM_Cubic>]}
       RedCurve {keys [<at 0 1 RCIM_Cubic>]}
      }})
    R1 Cobalt:LinearColor=&ColorFromCurve(ColorCurve=DataInterface() {propertyValues {
       AlphaCurve {keys [
         <at 0 0>
         {arriveTangent=0.3166, interpMode=RCIM_Cubic, leaveTangent=0.3166, time=0.2533, value=0.2}
         <at 0.6317 0.2>
         <at 0.9658 0 RCIM_Cubic>
        ]}
       BlueCurve {keys [<at 0 0.195 RCIM_Cubic>]}
       GreenCurve {keys [<at 0 0.0731 RCIM_Cubic>]}
       RedCurve {keys [<at 0 0.0487 RCIM_Cubic>]}
      }})
    R1 Crimson:LinearColor=&ColorFromCurve(ColorCurve=DataInterface() {propertyValues {
       AlphaCurve {keys [
         <at 0 0>
         {arriveTangent=0.3166, interpMode=RCIM_Cubic, leaveTangent=0.3166, time=0.2533, value=0.2}
         <at 0.6317 0.2>
         <at 0.9658 0 RCIM_Cubic>
        ]}
       BlueCurve {keys [<at 0 0.0167 RCIM_Cubic>]}
       GreenCurve {keys [<at 0 0 RCIM_Cubic>]}
       RedCurve {keys [<at 0 1 RCIM_Cubic>]}
      }})
    R1 DefaultColor=&ColorFromCurve(ColorCurve=DataInterface() {propertyValues {
       AlphaCurve {keys [
         <at 0 0>
         {arriveTangent=0.3166, interpMode=RCIM_Cubic, leaveTangent=0.3166, time=0.2533, value=0.2}
         <at 0.6317 0.2>
         <at 0.9658 0 RCIM_Cubic>
        ]}
       BlueCurve {keys [<at 0 0 RCIM_Cubic>]}
       GreenCurve {keys [<at 0 0.2667 RCIM_Cubic>]}
       RedCurve {keys [<at 0 1 RCIM_Cubic>]}
      }})
    R1 ForestGreen:LinearColor=&ColorFromCurve(ColorCurve=DataInterface() {propertyValues {
       AlphaCurve {keys [
         <at 0 0>
         {arriveTangent=0.3166, interpMode=RCIM_Cubic, leaveTangent=0.3166, time=0.2533, value=0.2}
         <at 0.6317 0.2>
         <at 0.9658 0 RCIM_Cubic>
        ]}
       BlueCurve {keys [<at 0 0.05 RCIM_Cubic>]}
       GreenCurve {keys [<at 0 1 RCIM_Cubic>]}
       RedCurve {keys [<at 0 0 RCIM_Cubic>]}
      }})
    R1 Grey:LinearColor=&ColorFromCurve(ColorCurve=DataInterface() {propertyValues {
       AlphaCurve {keys [
         <at 0 0>
         {arriveTangent=0.3166, interpMode=RCIM_Cubic, leaveTangent=0.3166, time=0.2533, value=0.2}
         <at 0.6317 0.2>
         <at 0.9658 0 RCIM_Cubic>
        ]}
       BlueCurve {keys [<at 0 0.1927 RCIM_Cubic>]}
       GreenCurve {keys [<at 0 0.1927 RCIM_Cubic>]}
       RedCurve {keys [<at 0 0.1927 RCIM_Cubic>]}
      }})
    R1 Lime:LinearColor=&ColorFromCurve(ColorCurve=DataInterface() {propertyValues {
       AlphaCurve {keys [
         <at 0 0>
         {arriveTangent=0.3166, interpMode=RCIM_Cubic, leaveTangent=0.3166, time=0.2533, value=0.2}
         <at 0.6317 0.2>
         <at 0.9658 0 RCIM_Cubic>
        ]}
       BlueCurve {keys [<at 0 0 RCIM_Cubic>]}
       GreenCurve {keys [<at 0 1 RCIM_Cubic>]}
       RedCurve {keys [<at 0 0.5833 RCIM_Cubic>]}
      }})
    R1 Orange:LinearColor=&ColorFromCurve(ColorCurve=DataInterface() {propertyValues {
       AlphaCurve {keys [
         <at 0 0>
         {arriveTangent=0.3166, interpMode=RCIM_Cubic, leaveTangent=0.3166, time=0.2533, value=0.2}
         <at 0.6317 0.2>
         <at 0.9658 0 RCIM_Cubic>
        ]}
       BlueCurve {keys [<at 0 0 RCIM_Cubic>]}
       GreenCurve {keys [<at 0 0.491 RCIM_Cubic>]}
       RedCurve {keys [<at 0 1 RCIM_Cubic>]}
      }})
    R1 Pink:LinearColor=&ColorFromCurve(ColorCurve=DataInterface() {propertyValues {
       AlphaCurve {keys [
         <at 0 0>
         {arriveTangent=0.3166, interpMode=RCIM_Cubic, leaveTangent=0.3166, time=0.2533, value=0.2}
         <at 0.6317 0.2>
         <at 0.9658 0 RCIM_Cubic>
        ]}
       BlueCurve {keys [<at 0 1 RCIM_Cubic>]}
       GreenCurve {keys [<at 0 0 RCIM_Cubic>]}
       RedCurve {keys [<at 0 0.8333 RCIM_Cubic>]}
      }})
    R1 Purple:LinearColor=&ColorFromCurve(ColorCurve=DataInterface() {propertyValues {
       AlphaCurve {keys [
         <at 0 0>
         {arriveTangent=0.3166, interpMode=RCIM_Cubic, leaveTangent=0.3166, time=0.2533, value=0.2}
         <at 0.6317 0.2>
         <at 0.9658 0 RCIM_Cubic>
        ]}
       BlueCurve {keys [<at 0 1 RCIM_Cubic>]}
       GreenCurve {keys [<at 0 0 RCIM_Cubic>]}
       RedCurve {keys [<at 0 0.4167 RCIM_Cubic>]}
      }})
    R1 Saffron:LinearColor=&ColorFromCurve(ColorCurve=DataInterface() {propertyValues {
       AlphaCurve {keys [
         <at 0 0>
         {arriveTangent=0.3166, interpMode=RCIM_Cubic, leaveTangent=0.3166, time=0.2533, value=0.2}
         <at 0.6317 0.2>
         <at 0.9658 0 RCIM_Cubic>
        ]}
       BlueCurve {keys [<at 0 0 RCIM_Cubic>]}
       GreenCurve {keys [<at 0 1 RCIM_Cubic>]}
       RedCurve {keys [<at 0 0.9167 RCIM_Cubic>]}
      }})
    R1 SkyBlue:LinearColor=&ColorFromCurve(ColorCurve=DataInterface() {propertyValues {
       AlphaCurve {keys [
         <at 0 0>
         {arriveTangent=0.3166, interpMode=RCIM_Cubic, leaveTangent=0.3166, time=0.2533, value=0.2}
         <at 0.6317 0.2>
         <at 0.9658 0 RCIM_Cubic>
        ]}
       BlueCurve {keys [<at 0 1 RCIM_Cubic>]}
       GreenCurve {keys [<at 0 0.4167 RCIM_Cubic>]}
       RedCurve {keys [<at 0 0 RCIM_Cubic>]}
      }})
    R1 TitaniumWhite:LinearColor=&ColorFromCurve(ColorCurve=DataInterface() {propertyValues {
       AlphaCurve {keys [
         <at 0 0>
         {arriveTangent=0.3166, interpMode=RCIM_Cubic, leaveTangent=0.3166, time=0.2533, value=0.2}
         <at 0.6317 0.2>
         <at 0.9658 0 RCIM_Cubic>
        ]}
       BlueCurve {keys [<at 0 0.1927 RCIM_Cubic>]}
       GreenCurve {keys [<at 0 0.1927 RCIM_Cubic>]}
       RedCurve {keys [<at 0 0.1927 RCIM_Cubic>]}
      }})
    Saffron=&ColorFromCurve(ColorCurve=DataInterface() {propertyValues {
       AlphaCurve {keys [<at 0 0>, {arriveTangent=1.583, interpMode=RCIM_Cubic, leaveTangent=1.583, time=0.2533, value=1}, <at 0.6317 1>, <at 0.9658 0 RCIM_Cubic>]}
       BlueCurve {keys [<at 0 0.02 RCIM_Cubic>]}
       GreenCurve {keys [<at 0 2 RCIM_Cubic>]}
       RedCurve {keys [<at 0 1.835 RCIM_Cubic>]}
      }})
    SkyBlue=&ColorFromCurve(ColorCurve=DataInterface() {propertyValues {
       AlphaCurve {keys [<at 0 0>, {arriveTangent=1.583, interpMode=RCIM_Cubic, leaveTangent=1.583, time=0.2533, value=1}, <at 0.6317 1>, <at 0.9658 0 RCIM_Cubic>]}
       BlueCurve {keys [<at 0 2 RCIM_Cubic>]}
       GreenCurve {keys [<at 0 0.9333 RCIM_Cubic>]}
       RedCurve {keys [<at 0 0 RCIM_Cubic>]}
      }})
    TitaniumWhite=&ColorFromCurve(ColorCurve=DataInterface() {propertyValues {
       AlphaCurve {keys [<at 0 0>, {arriveTangent=1.583, interpMode=RCIM_Cubic, leaveTangent=1.583, time=0.2533, value=1}, <at 0.6317 1>, <at 0.9658 0 RCIM_Cubic>]}
       BlueCurve {keys [<at 0 3 RCIM_Cubic>]}
       GreenCurve {keys [<at 0 3 RCIM_Cubic>]}
       RedCurve {keys [<at 0 3 RCIM_Cubic>]}
      }})
    Use R1 Painted Overrides=true
   ) {module=VC_PaintedVariantsOverride001}
   &FN_DisableTireFX(
    Mask Index Attribute:NiagaraInt32=$Particles.PartitionParticles.Partition
    Mask Index FL=2
    Mask Index FR=3
    Mask Index RL=0
    Mask Index RR=1
   )
   &InheritSourceMovement(ApplyForce=false, ApplyPosition=true, ApplyPositionFalloff=false)
   &ParticleState()
   &SolveForcesAndVelocity(Clamp Velocity=false, Limit Acceleration=false)
  ]
  rendererClasses [&NiagaraSpriteRendererProperties]
  renderers [
   {rendererClass=&NiagaraSpriteRendererProperties}
   {rendererClass=&NiagaraSpriteRendererProperties, rendererIndex=1}
  ]
 }
 NE_VC_Trail {
  emitterUpdateScript [
   &EmitterState(Inactive Response=::NewEnumerator1, Life Cycle Mode=::NewEnumerator1, Loop Duration=1, UseLoopDelay=false)
   &VC_RateQualityHandler(High Spawn Rate=40, Low Spawn Rate=30, Use Custom Rate?=true)
   &VC_SpawnRibbon_v1(Number of Ribbons=2, SpawnRate=$Emitter.Spawn Rate)
  ]
  particleSpawnScript [
   &VC_PaintedVariantsOverride(Use R1 Painted Overrides=true) {enabled=false}
   &VC_InitializeRibbon_v1(
    Custom Offset=true
    Enable Skeletal Position=true
    Lifetime=0.75
    Local Vector Offset=&AddVectorToPosition(Position=<-100 0 -10>, Vector=$User.offset)
    Offset Amount=1
    Ribbon Width=45
   )
   &AddVelocity(Velocity=<-1600 0 0>) {module=AddVelocity001}
  ]
  particleUpdateScript [
   &VC_PaintedVariantsOverride(
    Black=&ColorFromCurve(ColorCurve=DataInterface() {propertyValues {
       AlphaCurve {keys [<at 0 0.7>, <at 0.15 1>, {arriveTangent=-2.1277, interpMode=RCIM_Cubic, leaveTangent=-2.1277, time=0.31, value=1}, <at 0.62 0>]}
       BlueCurve {keys [<at 0 0.007 RCIM_Cubic>]}
       GreenCurve {keys [<at 0 0.007 RCIM_Cubic>]}
       RedCurve {keys [<at 0 0.007 RCIM_Cubic>]}
      }})
    BurntSienna=&ColorFromCurve(ColorCurve=DataInterface() {propertyValues {
       AlphaCurve {keys [<at 0 0.7>, <at 0.15 1>, {arriveTangent=-2.1277, interpMode=RCIM_Cubic, leaveTangent=-2.1277, time=0.31, value=1}, <at 0.62 0>]}
       BlueCurve {keys [<at 0 0.001>]}
       GreenCurve {keys [<at 0 0.0073>]}
       RedCurve {keys [<at 0 0.02>]}
      }})
    Cobalt=&ColorFromCurve(ColorCurve=DataInterface() {propertyValues {
       AlphaCurve {keys [<at 0 0.7>, <at 0.15 1>, {arriveTangent=-2.1277, interpMode=RCIM_Cubic, leaveTangent=-2.1277, time=0.31, value=1}, <at 0.62 0>]}
       BlueCurve {keys [<at 0 0.15>]}
       GreenCurve {keys [<at 0 0.0235>]}
       RedCurve {keys [<at 0 0.012>]}
      }})
    Crimson=&ColorFromCurve(ColorCurve=DataInterface() {propertyValues {
       AlphaCurve {keys [<at 0 0.7>, <at 0.15 1>, {arriveTangent=-2.1277, interpMode=RCIM_Cubic, leaveTangent=-2.1277, time=0.31, value=1}, <at 0.62 0>]}
       BlueCurve {keys [<at 0 0.0032>]}
       GreenCurve {keys [<at 0 0.0021>]}
       RedCurve {keys [<at 0 0.07>]}
      }})
    DefaultColor=&ColorFromCurve(ColorCurve=DataInterface() {propertyValues {
       AlphaCurve {keys [<at 0 0.7>, <at 0.15 1>, {arriveTangent=-2.1277, interpMode=RCIM_Cubic, leaveTangent=-2.1277, time=0.31, value=1}, <at 0.62 0>]}
       BlueCurve {keys [<at 0 0.2812 RCIM_Cubic>]}
       GreenCurve {keys [<at 0 0.0659 RCIM_Cubic>]}
       RedCurve {keys [<at 0 0.0659 RCIM_Cubic>]}
      }})
    ForestGreen=&ColorFromCurve(ColorCurve=DataInterface() {propertyValues {
       AlphaCurve {keys [<at 0 0.7>, <at 0.15 1>, {arriveTangent=-2.1277, interpMode=RCIM_Cubic, leaveTangent=-2.1277, time=0.31, value=1}, <at 0.62 0>]}
       BlueCurve {keys [<at 0 0.0043>]}
       GreenCurve {keys [<at 0 0.05>]}
       RedCurve {keys [<at 0 0.001>]}
      }})
    Grey=&ColorFromCurve(ColorCurve=DataInterface() {propertyValues {
       AlphaCurve {keys [<at 0 0.7>, <at 0.15 1>, {arriveTangent=-1.129, interpMode=RCIM_Cubic, leaveTangent=-1.129, time=0.31, value=1}, <at 0.62 0>]}
       BlueCurve {keys [<at 0 0.03 RCIM_Cubic>]}
       GreenCurve {keys [<at 0 0.03 RCIM_Cubic>]}
       RedCurve {keys [<at 0 0.03 RCIM_Cubic>]}
      }})
    Lime=&ColorFromCurve(ColorCurve=DataInterface() {propertyValues {
       AlphaCurve {keys [<at 0 0.7>, <at 0.15 1>, {arriveTangent=-2.1277, interpMode=RCIM_Cubic, leaveTangent=-2.1277, time=0.31, value=1}, <at 0.62 0>]}
       BlueCurve {keys [<at 0 0.0022 RCIM_Cubic>]}
       GreenCurve {keys [<at 0 0.11 RCIM_Cubic>]}
       RedCurve {keys [<at 0 0.0651 RCIM_Cubic>]}
      }})
    Orange=&ColorFromCurve(ColorCurve=DataInterface() {propertyValues {
       AlphaCurve {keys [<at 0 0.7>, <at 0.15 1>, {arriveTangent=-2.1277, interpMode=RCIM_Cubic, leaveTangent=-2.1277, time=0.31, value=1}, <at 0.62 0>]}
       BlueCurve {keys [<at 0 0.004 RCIM_Cubic>]}
       GreenCurve {keys [<at 0 0.028 RCIM_Cubic>]}
       RedCurve {keys [<at 0 0.0999 RCIM_Cubic>]}
      }})
    Pink=&ColorFromCurve(ColorCurve=DataInterface() {propertyValues {
       AlphaCurve {keys [<at 0 0.7>, <at 0.15 1>, {arriveTangent=-2.1277, interpMode=RCIM_Cubic, leaveTangent=-2.1277, time=0.31, value=1}, <at 0.62 0>]}
       BlueCurve {keys [<at 0 0.08 RCIM_Cubic>]}
       GreenCurve {keys [<at 0 0.028 RCIM_Cubic>]}
       RedCurve {keys [<at 0 0.0765 RCIM_Cubic>]}
      }})
    Purple=&ColorFromCurve(ColorCurve=DataInterface() {propertyValues {
       AlphaCurve {keys [<at 0 0.7>, <at 0.15 1>, {arriveTangent=-2.1277, interpMode=RCIM_Cubic, leaveTangent=-2.1277, time=0.31, value=1}, <at 0.62 0>]}
       BlueCurve {keys [<at 0 0.05 RCIM_Cubic>]}
       GreenCurve {keys [<at 0 0.001 RCIM_Cubic>]}
       RedCurve {keys [<at 0 0.0214 RCIM_Cubic>]}
      }})
    R1 Black:LinearColor=&ColorFromCurve(ColorCurve=DataInterface() {propertyValues {
       AlphaCurve {keys [
         <at 0 0.2>
         <at 0.1 1>
         {arriveTangent=-1.9231, interpMode=RCIM_Cubic, leaveTangent=-1.9231, time=0.3, value=1}
         {interpMode=RCIM_Cubic, leaveTangent=-1.3699, time=0.62, value=0}
        ]}
       BlueCurve {keys [<at 0 0.03 RCIM_Cubic>]}
       GreenCurve {keys [<at 0 0.03 RCIM_Cubic>]}
       RedCurve {keys [<at 0 0.03 RCIM_Cubic>]}
      }})
    R1 BurntSienna:LinearColor=&ColorFromCurve(ColorCurve=DataInterface() {propertyValues {
       AlphaCurve {keys [
         <at 0 0.2>
         <at 0.1 1>
         {arriveTangent=-1.9231, interpMode=RCIM_Cubic, leaveTangent=-1.9231, time=0.3, value=1}
         {interpMode=RCIM_Cubic, leaveTangent=-1.3699, time=0.62, value=0}
        ]}
       BlueCurve {keys [<at 0 0.002 RCIM_Cubic>]}
       GreenCurve {keys [<at 0 0.0249 RCIM_Cubic>]}
       RedCurve {keys [<at 0 0.1 RCIM_Cubic>]}
      }})
    R1 Cobalt:LinearColor=&ColorFromCurve(ColorCurve=DataInterface() {propertyValues {
       AlphaCurve {keys [
         <at 0 0.2>
         <at 0.1 1>
         {arriveTangent=-1.9231, interpMode=RCIM_Cubic, leaveTangent=-1.9231, time=0.3, value=1}
         {interpMode=RCIM_Cubic, leaveTangent=-1.3699, time=0.62, value=0}
        ]}
       BlueCurve {keys [<at 0 0.5 RCIM_Cubic>]}
       GreenCurve {keys [<at 0 0.0508 RCIM_Cubic>]}
       RedCurve {keys [<at 0 0.01 RCIM_Cubic>]}
      }})
    R1 Crimson:LinearColor=&ColorFromCurve(ColorCurve=DataInterface() {propertyValues {
       AlphaCurve {keys [<at 0 0.2>, <at 0.1 1>, {arriveTangent=-1.9231, interpMode=RCIM_Cubic, leaveTangent=-1.9231, time=0.31, value=1}, <at 0.62 0>]}
       BlueCurve {keys [<at 0 0.0094>]}
       GreenCurve {keys [<at 0 0.0052>]}
       RedCurve {keys [<at 0 0.26>]}
      }})
    R1 DefaultColor=&ColorFromCurve(ColorCurve=DataInterface() {propertyValues {
       AlphaCurve {keys [
         <at 0 0.1>
         <at 0.15 1>
         {arriveTangent=-2.1277, interpMode=RCIM_Cubic, leaveTangent=-2.1277, time=0.3, value=1}
         {interpMode=RCIM_Cubic, leaveTangent=-1.3699, time=0.62, value=0}
        ]}
       BlueCurve {keys [<at 0 0.8957 RCIM_Cubic>]}
       GreenCurve {keys [<at 0 0.513 RCIM_Cubic>]}
       RedCurve {keys [<at 0 0.0255 RCIM_Cubic>]}
      }})
    R1 ForestGreen:LinearColor=&ColorFromCurve(ColorCurve=DataInterface() {propertyValues {
       AlphaCurve {keys [<at 0 0.2>, <at 0.1 1>, {arriveTangent=-1.9231, interpMode=RCIM_Cubic, leaveTangent=-1.9231, time=0.31, value=1}, <at 0.62 0>]}
       BlueCurve {keys [<at 0 0.0071>]}
       GreenCurve {keys [<at 0 0.12>]}
       RedCurve {keys [<at 0 0.0012>]}
      }})
    R1 Grey:LinearColor=&ColorFromCurve(ColorCurve=DataInterface() {propertyValues {
       AlphaCurve {keys [<at 0 0.2>, <at 0.1 1>, {arriveTangent=-1.9231, interpMode=RCIM_Cubic, leaveTangent=-1.9231, time=0.31, value=1}, <at 0.62 0>]}
       BlueCurve {keys [<at 0 0.1>]}
       GreenCurve {keys [<at 0 0.1>]}
       RedCurve {keys [<at 0 0.1>]}
      }})
    R1 Lime:LinearColor=&ColorFromCurve(ColorCurve=DataInterface() {propertyValues {
       AlphaCurve {keys [
         <at 0 0.2>
         <at 0.1 1>
         {arriveTangent=-1.9231, interpMode=RCIM_Cubic, leaveTangent=-1.9231, time=0.3, value=1}
         {interpMode=RCIM_Cubic, leaveTangent=-1.3699, time=0.62, value=0}
        ]}
       BlueCurve {keys [<at 0 0.006 RCIM_Cubic>]}
       GreenCurve {keys [<at 0 0.3 RCIM_Cubic>]}
       RedCurve {keys [<at 0 0.1628 RCIM_Cubic>]}
      }})
    R1 Orange:LinearColor=&ColorFromCurve(ColorCurve=DataInterface() {propertyValues {
       AlphaCurve {keys [
         <at 0 0.2>
         <at 0.1 1>
         {arriveTangent=-1.9231, interpMode=RCIM_Cubic, leaveTangent=-1.9231, time=0.3, value=1}
         {interpMode=RCIM_Cubic, leaveTangent=-1.3699, time=0.62, value=0}
        ]}
       BlueCurve {keys [<at 0 0.006 RCIM_Cubic>]}
       GreenCurve {keys [<at 0 0.0844 RCIM_Cubic>]}
       RedCurve {keys [<at 0 0.3 RCIM_Cubic>]}
      }})
    R1 Pink:LinearColor=&ColorFromCurve(ColorCurve=DataInterface() {propertyValues {
       AlphaCurve {keys [
         <at 0 0.2>
         <at 0.1 1>
         {arriveTangent=-1.9231, interpMode=RCIM_Cubic, leaveTangent=-1.9231, time=0.3, value=1}
         {interpMode=RCIM_Cubic, leaveTangent=-1.3699, time=0.62, value=0}
        ]}
       BlueCurve {keys [<at 0 0.3 RCIM_Cubic>]}
       GreenCurve {keys [<at 0 0.09 RCIM_Cubic>]}
       RedCurve {keys [<at 0 0.293 RCIM_Cubic>]}
      }})
    R1 Purple:LinearColor=&ColorFromCurve(ColorCurve=DataInterface() {propertyValues {
       AlphaCurve {keys [
         <at 0 0.2>
         <at 0.1 1>
         {arriveTangent=-1.9231, interpMode=RCIM_Cubic, leaveTangent=-1.9231, time=0.3, value=1}
         {interpMode=RCIM_Cubic, leaveTangent=-1.3699, time=0.62, value=0}
        ]}
       BlueCurve {keys [<at 0 0.3 RCIM_Cubic>]}
       GreenCurve {keys [<at 0 0.006 RCIM_Cubic>]}
       RedCurve {keys [<at 0 0.1285 RCIM_Cubic>]}
      }})
    R1 Saffron:LinearColor=&ColorFromCurve(ColorCurve=DataInterface() {propertyValues {
       AlphaCurve {keys [
         <at 0 0.2>
         <at 0.1 1>
         {arriveTangent=-1.9231, interpMode=RCIM_Cubic, leaveTangent=-1.9231, time=0.3, value=1}
         {interpMode=RCIM_Cubic, leaveTangent=-1.3699, time=0.62, value=0}
        ]}
       BlueCurve {keys [<at 0 0.006 RCIM_Cubic>]}
       GreenCurve {keys [<at 0 0.3 RCIM_Cubic>]}
       RedCurve {keys [<at 0 0.3 RCIM_Cubic>]}
      }})
    R1 SkyBlue:LinearColor=&ColorFromCurve(ColorCurve=DataInterface() {propertyValues {
       AlphaCurve {keys [
         <at 0 0.2>
         <at 0.1 1>
         {arriveTangent=-0.3226, interpMode=RCIM_Cubic, leaveTangent=-0.3226, time=0.3, value=1}
         {interpMode=RCIM_Cubic, leaveTangent=-1.3699, time=0.62, value=0}
        ]}
       BlueCurve {keys [<at 0 0.4 RCIM_Cubic>]}
       GreenCurve {keys [<at 0 0.202 RCIM_Cubic>]}
       RedCurve {keys [<at 0 0.004 RCIM_Cubic>]}
      }})
    R1 TitaniumWhite:LinearColor=&ColorFromCurve(ColorCurve=DataInterface() {propertyValues {
       AlphaCurve {keys [
         <at 0 0.2>
         <at 0.1 1>
         {arriveTangent=-0.3226, interpMode=RCIM_Cubic, leaveTangent=-0.3226, time=0.3, value=1}
         {interpMode=RCIM_Cubic, leaveTangent=-1.3699, time=0.62, value=0}
        ]}
       BlueCurve {keys [<at 0 0.3 RCIM_Cubic>]}
       GreenCurve {keys [<at 0 0.3 RCIM_Cubic>]}
       RedCurve {keys [<at 0 0.3 RCIM_Cubic>]}
      }})
    R2 Black:LinearColor=<rgba 0.1 0.1 0.1 1.3>
    R2 BurntSienna:LinearColor=<rgba 0.55 0.1457 0.011 1.3>
    R2 Cobalt:LinearColor=<rgba 0.036 0.0936 0.9 1>
    R2 Crimson:LinearColor=<rgba 0.3 0.006 0.0109 1>
    R2 DefaultColor:LinearColor=<rgba 0.076 0.6587 0.95 1.3>
    R2 ForestGreen:LinearColor=<rgba 0.01 0.5 0.0345 1>
    R2 Grey:LinearColor=<rgba 0.35 0.35 0.35 1>
    R2 Lime:LinearColor=<rgba 0.4942 0.7 0.014 1>
    R2 Orange:LinearColor=<rgba 0.8 0.3296 0.016 1>
    R2 Pink:LinearColor=<rgba 0.6563 0.175 0.7 1>
    R2 Purple:LinearColor=<rgba 0.408 0.016 0.8 1>
    R2 Saffron:LinearColor=<rgba 0.5046 0.55 0.0055 1>
    R2 SkyBlue:LinearColor=<rgba 0.007 0.3535 0.7 1>
    R2 TitaniumWhite:LinearColor=<rgba 0.6 0.6 0.6 1>
    R3 Black:LinearColor=<rgba 0.23 0.23 0.23 0.5>
    R3 BurntSienna:LinearColor=<rgba 0.55 0.2176 0.011 0.5>
    R3 Cobalt:LinearColor=<rgba 0.1634 0.2679 0.86 0.5>
    R3 Crimson:LinearColor=<rgba 0.55 0.0055 0.0146 0.5>
    R3 DefaultColor:LinearColor=<rgba 0.85 0.6957 0.002 0.5>
    R3 ForestGreen:LinearColor=<rgba 0.012 0.6 0.0512 0.5>
    R3 Grey:LinearColor=<rgba 0.4 0.4 0.4 0.5>
    R3 Lime:LinearColor=<rgba 0.604 0.8 0.016 0.5>
    R3 Orange:LinearColor=<rgba 0.8 0.408 0.016 0.5>
    R3 Pink:LinearColor=<rgba 0.7067 0.24 0.8 0.5>
    R3 Purple:LinearColor=<rgba 0.3838 0.0076 0.76 0.5>
    R3 Saffron:LinearColor=<rgba 0.734 0.8 0.008 0.5>
    R3 SkyBlue:LinearColor=<rgba 0.016 0.408 0.8 0.5>
    R3 TitaniumWhite:LinearColor=<rgba 0.6 0.6 0.6 0.5>
    Saffron=&ColorFromCurve(ColorCurve=DataInterface() {propertyValues {
       AlphaCurve {keys [<at 0 0.7>, <at 0.15 1>, {arriveTangent=-2.1277, interpMode=RCIM_Cubic, leaveTangent=-2.1277, time=0.31, value=1}, <at 0.62 0>]}
       BlueCurve {keys [<at 0 0.02 RCIM_Cubic>]}
       GreenCurve {keys [<at 0 0.1 RCIM_Cubic>]}
       RedCurve {keys [<at 0 0.1 RCIM_Cubic>]}
      }})
    SkyBlue=&ColorFromCurve(ColorCurve=DataInterface() {propertyValues {
       AlphaCurve {keys [<at 0 0.7>, <at 0.15 1>, {arriveTangent=-2.1277, interpMode=RCIM_Cubic, leaveTangent=-2.1277, time=0.31, value=1}, <at 0.62 0>]}
       BlueCurve {keys [<at 0 0.15 RCIM_Cubic>]}
       GreenCurve {keys [<at 0 0.0634 RCIM_Cubic>]}
       RedCurve {keys [<at 0 0.0015 RCIM_Cubic>]}
      }})
    TitaniumWhite=&ColorFromCurve(ColorCurve=DataInterface() {propertyValues {
       AlphaCurve {keys [<at 0 0.7>, <at 0.15 1>, {arriveTangent=-2.1277, interpMode=RCIM_Cubic, leaveTangent=-2.1277, time=0.31, value=1}, <at 0.62 0>]}
       BlueCurve {keys [<at 0 0.11 RCIM_Cubic>]}
       GreenCurve {keys [<at 0 0.11 RCIM_Cubic>]}
       RedCurve {keys [<at 0 0.11 RCIM_Cubic>]}
      }})
    Use R1 Painted Overrides=true
    Use R2 Painted Overrides=true
    Use R3 Painted Overrides=true
   ) {module=VC_PaintedVariantsOverride001}
   &FN_DisableTireFX(Mask Index Attribute:NiagaraInt32=$Particles.SpawnGroup, Mask Index FL=2, Mask Index FR=3, Mask Index RL=0, Mask Index RR=1)
   &ParticleState()
   &InheritSourceMovement(ApplyForce=false, ApplyPosition=true, ApplyPositionFalloff=false)
   &SolveForcesAndVelocity(Clamp Velocity=false, Limit Acceleration=false)
   &VC_UpdateRibbon_v1(Enable Skeletal Position=true)
   &ScaleRibbonWidth(Ribbon Width Scale=&FloatFromCurve(FloatCurve=DataInterface() {propertyValues {Curve {keys [
         {leaveTangentWeight=0.2451, time=0, value=1.6}
         <at 0.3 1>
         {arriveTangentWeight=0.3466, interpMode=RCIM_Cubic, leaveTangentWeight=0.1248, tangentMode=RCTM_User, time=0.6, value=1}
         {arriveTangent=-1.5734, arriveTangentWeight=0.0883, interpMode=RCIM_Cubic, leaveTangent=-1.5734, tangentMode=RCTM_User, time=1, value=0.5}
        ]}}}))
   /SetVariables(
    Particles.glowwidth:NiagaraFloat=&Multiply_Float(
     A=&FloatFromCurve(FloatCurve=DataInterface() {propertyValues {Curve {keys [
          {arriveTangent=-0.1858, interpMode=RCIM_Cubic, leaveTangent=-0.1858, tangentMode=RCTM_User, time=0, value=0.5}
          {arriveTangent=4.91, interpMode=RCIM_Cubic, leaveTangent=4.91, tangentMode=RCTM_User, time=0.2, value=1}
         ]}}})
     B=50
    )
    Particles.goldwidth:NiagaraFloat=&Multiply_Float(
     A=&FloatFromCurve(FloatCurve=DataInterface() {propertyValues {Curve {keys [
          {arriveTangent=0.9575, interpMode=RCIM_Cubic, leaveTangent=0.9575, leaveTangentWeight=0.0662, tangentMode=RCTM_User, time=0, value=0.3}
          {arriveTangent=0.3593, interpMode=RCIM_Cubic, leaveTangent=0.3593, tangentMode=RCTM_User, time=0.395, value=1.3917}
          {arriveTangent=-1.6657, interpMode=RCIM_Cubic, leaveTangent=-1.6657, tangentMode=RCTM_User, time=1, value=0.5}
         ]}}})
     B=46
    )
   ) {isSetParameters=true}
   /SetVariables(Particles.glowposition:NiagaraPosition=&AddVectorToPosition(Position=$Particles.Position, Vector=<0 0 -2>)) {isSetParameters=true}
  ]
  rendererClasses [&NiagaraRibbonRendererProperties]
  renderers [
   {rendererClass=&NiagaraRibbonRendererProperties}
   {rendererClass=&NiagaraRibbonRendererProperties, rendererIndex=1}
   {rendererClass=&NiagaraRibbonRendererProperties, rendererIndex=2}
   {rendererClass=&NiagaraRibbonRendererProperties, rendererIndex=3}
  ]
 }
 NE_VC_WheelWrap {
  emitterUpdateScript [
   &EmitterState(
    Inactive Response=::NewEnumerator1
    Life Cycle Mode=::NewEnumerator1
    Loop Behavior=::NewEnumerator1
    Loop Duration Mode=::NewEnumerator1
    UseLoopDelay=false
   )
   &SpawnBurst_Instantaneous(Spawn Count=2, Use Spawn Probability=false, UseLoopCountLimit=false)
  ]
  particleSpawnScript [
   &VC_PaintedVariantsOverride(
    Black=<rgba 0.07 0.07 0.07 1>
    BurntSienna=<rgba 0.23 0.0497 0.0046 1>
    Cobalt=<rgba 0.05 0.1175 0.5 1>
    Crimson=<rgba 0.4 0.008 0.0145 1>
    DefaultColor=<rgba 0.599 0.2791 0.0499 1>
    ForestGreen=<rgba 0.006 0.3 0.0207 1>
    Grey=<rgba 0.2 0.2 0.2 1>
    Lime=<rgba 0.3775 0.5 0.01 1>
    Orange=<rgba 0.4 0.1909 0.008 1>
    Pink=<rgba 0.3533 0.12 0.4 1>
    Purple=<rgba 0.2152 0.0205 0.41 1>
    R1 Black:LinearColor=<rgba 0.08 0.08 0.08 1>
    R1 BurntSienna:LinearColor=<rgba 0.17 0.0476 0.0068 1>
    R1 Cobalt:LinearColor=<rgba 0.07 0.112 0.7 1>
    R1 Crimson:LinearColor=<rgba 0.45 0.009 0.0163 1>
    R1 DefaultColor=<rgba 0.2626 1.0685 2 2>
    R1 ForestGreen:LinearColor=<rgba 0.0042 0.21 0.0179 1>
    R1 Grey:LinearColor=<rgba 0.16 0.16 0.16 1>
    R1 Lime:LinearColor=<rgba 0.2958 0.5 0.01 1>
    R1 Orange:LinearColor=<rgba 0.6 0.1688 0.012 1>
    R1 Pink:LinearColor=<rgba 0.565 0.18 0.6 1>
    R1 Purple:LinearColor=<rgba 0.257 0.012 0.6 1>
    R1 Saffron:LinearColor=<rgba 0.4592 0.5 0.01 1>
    R1 SkyBlue:LinearColor=<rgba 0.012 0.306 0.6 1>
    R1 TitaniumWhite:LinearColor=<rgba 0.5 0.5 0.5 1>
    R2 Black:LinearColor=<rgba 0.006 0.006 0.006 0.8>
    R2 BurntSienna:LinearColor=<rgba 0.04 0.008 0.0008 0.4>
    R2 Cobalt:LinearColor=<rgba 0.039 0.0607 0.3 0.4>
    R2 Crimson:LinearColor=<rgba 0.2 0.004 0.0073 0.4>
    R2 DefaultColor:LinearColor=<rgba 0.1556 0.2137 0.8522 0.2>
    R2 ForestGreen:LinearColor=<rgba 0.001 0.1 0.0076 0.4>
    R2 Grey:LinearColor=<rgba 0.05 0.05 0.05 0.5>
    R2 Lime:LinearColor=<rgba 0.1175 0.2 0.002 0.4>
    R2 Orange:LinearColor=<rgba 0.2 0.0563 0.004 0.4>
    R2 Pink:LinearColor=<rgba 0.1883 0.06 0.2 0.4>
    R2 Purple:LinearColor=<rgba 0.0514 0.0024 0.12 0.4>
    R2 Saffron:LinearColor=<rgba 0.1835 0.2 0.002 0.4>
    R2 SkyBlue:LinearColor=<rgba 0.002 0.101 0.2 0.4>
    R2 TitaniumWhite:LinearColor=<rgba 0.25 0.25 0.25 0.4>
    Saffron=<rgba 0.4592 0.5 0.01 1>
    SkyBlue=<rgba 0.01 0.2142 0.5 1>
    TitaniumWhite=<rgba 0.41 0.41 0.41 1>
    Use R1 Painted Overrides=true
    Use R2 Painted Overrides=true
   )
   &InitializeParticle(Color Mode=::NewEnumerator0, Mesh Scale Mode=::NewEnumerator4, UsePositionOffset=false, Write Material Random=true)
   &PartitionParticles()
   &SkeletalMeshLocation(
    Bone / Socket Index=&Add_Integer(A=$Particles.PartitionParticles.Partition, B=4)
    Bone / Socket Sampling Mode=::NewEnumerator5
    Interpolate Mesh Sampling=false
    MaskSpawnGroup=false
    Mesh Sampling Type=::NewEnumerator6
    Skeletal Mesh=$System.SkeletalMesh
   )
   /SetVariables(Particles.VisibilityTag=$Particles.PartitionParticles.Partition) {isSetParameters=true}
  ]
  particleUpdateScript [
   &FN_DisableTireFX(
    Mask Index Attribute:NiagaraInt32=$Particles.PartitionParticles.Partition
    Mask Index FL=2
    Mask Index FR=3
    Mask Index RL=0
    Mask Index RR=1
   )
   &VC_WheelWrapScale(Wheel Index Attribute:NiagaraInt32=&Add_Integer(A=$Particles.PartitionParticles.Partition, B=2))
   &ParticleState(Kill Particles When Lifetime Has Elapsed=false)
   &UpdateMeshOrientation(Rotation Coordinate Space=::NewEnumerator2, Rotation Vector=<0 1 0>) {enabled=false}
   &SkeletalMeshLocation(
    Bone / Socket Index=&Add_Integer(A=$Particles.PartitionParticles.Partition, B=4)
    Bone / Socket Sampling Mode=::NewEnumerator5
    MaskSpawnGroup=false
    Mesh Sampling Type=::NewEnumerator6
    Skeletal Mesh=$System.SkeletalMesh
   ) {module=SkeletalMeshLocation001}
   &SolveForcesAndVelocity(Clamp Velocity=false, Limit Acceleration=false)
   &ScaleColor(Scale Alpha=2, Scale RGB=&VectorFromFloat(Value=1.5), ScaleA=true, ScaleRGB=true)
  ]
  rendererClasses [&NiagaraMeshRendererProperties]
  renderers [
   {rendererClass=&NiagaraMeshRendererProperties, rendererIndex=0}
   {rendererClass=&NiagaraMeshRendererProperties, rendererIndex=1}
   {rendererClass=&NiagaraMeshRendererProperties, rendererIndex=2}
   {rendererClass=&NiagaraMeshRendererProperties, rendererIndex=3}
   {rendererClass=&NiagaraMeshRendererProperties, rendererIndex=4}
   {rendererClass=&NiagaraMeshRendererProperties, rendererIndex=5}
  ]
 }
}
summary {
 systemName=NS_DriftTrail_DecoBlue_MP
 emitters [
  {emitterName=NE_VC_WheelWrap, rendererClasses [&NiagaraMeshRendererProperties]}
  {emitterName=NE_VC_Cone, rendererClasses [&NiagaraMeshRendererProperties, &NiagaraSpriteRendererProperties]}
  {emitterName=NE_VC_Trail, rendererClasses [&NiagaraRibbonRendererProperties]}
  {emitterName=NE_VC_Sprite, rendererClasses [&NiagaraSpriteRendererProperties]}
  {emitterName=NE_VC_Sprite001, rendererClasses [&NiagaraSpriteRendererProperties]}
 ]
 userVariables [
  User.bIsFrontendPreview {type=NiagaraBool}
  User.bIsPainted {type=NiagaraBool}
  User.ColorVariantID {type=NiagaraInt32}
  User.GameModeID {type=NiagaraInt32}
  User.offset {type=Vector3f, description="offset from wheels(set through cosmetic template)", defaultValue {struct=Vector3f, value=<0 0 0>}}
  User.PaintedColor {type=LinearColor, description="", defaultValue {struct=LinearColor, value=<rgba 1 1 1 1>}}
  User.PaintedScalar {type=NiagaraFloat, defaultValue=1, description=""}
  User.VehicleFlags {type=NiagaraInt32, defaultValue=806289408}
  User.Wheel_Radius_Array {
   type=NiagaraDataInterfaceArrayFloat
   defaultValue {
    struct=NiagaraExt_VariableValue_DataInterface
    value {
     dataInterface=&self:NiagaraDataInterfaceArrayFloat_1
     dataInterfaceClass=&NiagaraDataInterfaceArrayFloat
    }
   }
  }
  User.Wheel_Width_Array {
   type=NiagaraDataInterfaceArrayFloat
   defaultValue {
    struct=NiagaraExt_VariableValue_DataInterface
    value {
     dataInterface=&self:NiagaraDataInterfaceArrayFloat_0
     dataInterfaceClass=&NiagaraDataInterfaceArrayFloat
    }
   }
  }
  User.WheelTouchingArray {
   type=NiagaraDataInterfaceArrayFloat
   defaultValue {
    struct=NiagaraExt_VariableValue_DataInterface
    value {
     dataInterface=&self:NiagaraDataInterfaceArrayFloat_2
     dataInterfaceClass=&NiagaraDataInterfaceArrayFloat
    }
   }
  }
 ]
}
